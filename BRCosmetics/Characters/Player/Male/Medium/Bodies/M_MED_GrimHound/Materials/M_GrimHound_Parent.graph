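# Parent material for the GrimHound male player body cosmetic, adding a selectable accent color to the shared 2023 character shading.
# MF_CharacterParent_2023_PreFX supplies the base material attributes, and MF_CharacterParent_2023_PostFX processes them after the color edit.
# Between them, the FX mask's green channel lerps Base Color toward an accent color parameter.
# A Switch on ColorAccent_Index picks red, yellow, blue or beige, with red as the fallback.
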Material /BRCosmetics/Characters/Player/Male/Medium/Bodies/M_MED_GrimHound/Materials/M_GrimHound_Parent
graph {
 expressions {
  MaterialExpressionComponentMask_0 {class=MaterialExpressionComponentMask, inputs [None {from=MaterialExpressionTextureSampleParameter2D_0, output=RGB}], properties {g=true}}
  MaterialExpressionGetMaterialAttributes_0 {
   class=MaterialExpressionGetMaterialAttributes
   inputs [None {from=MaterialExpressionMaterialFunctionCall_0, output=Result}]
   properties {attributeGetTypes [#1]}
  }
  MaterialExpressionLinearInterpolate_1 {
   class=MaterialExpressionLinearInterpolate
   inputs [
    A {from=MaterialExpressionGetMaterialAttributes_0, output=Base Color}
    B {from=MaterialExpressionSwitch_0}
    Alpha {from=MaterialExpressionComponentMask_0}
   ]
  }
  MaterialExpressionMaterialFunctionCall_0 {
   class=MaterialExpressionMaterialFunctionCall
   inputs [
    Overide Texture Set UVs {from=""}
    Override Tech Art Mask UVs {from=""}
    Override Basic Emissive UVs {from=""}
   ]
   properties {materialFunction=&MF_CharacterParent_2023_PreFX}
  }
  MaterialExpressionMaterialFunctionCall_1 {
   class=MaterialExpressionMaterialFunctionCall
   inputs [
    Material {from=MaterialExpressionSetMaterialAttributes_0}
    M {from=MaterialExpressionMaterialFunctionCall_0, output=M}
    Pre Skinned Local Position {from=MaterialExpressionMaterialFunctionCall_0, output=M}
    Distance Blend {from=MaterialExpressionMaterialFunctionCall_0, output=M}
    Fresnel {from=MaterialExpressionMaterialFunctionCall_0, output=M}
   ]
   properties {materialFunction=&MF_CharacterParent_2023_PostFX}
  }
  MaterialExpressionScalarParameter_0 {class=MaterialExpressionScalarParameter, properties {group=GrimHound, parameterName=ColorAccent_Index}}
  MaterialExpressionSetMaterialAttributes_0 {
   class=MaterialExpressionSetMaterialAttributes
   inputs [
    MaterialAttributes {from=MaterialExpressionGetMaterialAttributes_0, output=MaterialAttributes}
    Base Color {from=MaterialExpressionLinearInterpolate_1, output=""}
   ]
   properties {attributeSetTypes [#1]}
  }
  MaterialExpressionSwitch_0 {
   class=MaterialExpressionSwitch
   inputs [
    SwitchValue {from=MaterialExpressionScalarParameter_0, output=""}
    Default {from=MaterialExpressionVectorParameter_3, output=RGB}
    None {from=MaterialExpressionVectorParameter_3, output=RGB}
    None {from=MaterialExpressionVectorParameter_0, output=RGB}
    None {from=MaterialExpressionVectorParameter_1, output=RGB}
    None {from=MaterialExpressionVectorParameter_2, output=RGB}
   ]
   properties {inputs [
     {
      inputName=None
      input {
       expression=&self:MaterialExpressionVectorParameter_3
       inputName=None
       mask=1
       maskA=0
       maskB=1
       maskG=1
       maskR=1
       outputIndex=0
      }
     }
     {
      inputName=None
      input {
       expression=&self:MaterialExpressionVectorParameter_0
       inputName=None
       mask=1
       maskA=0
       maskB=1
       maskG=1
       maskR=1
       outputIndex=0
      }
     }
     {
      inputName=None
      input {
       expression=&self:MaterialExpressionVectorParameter_1
       inputName=None
       mask=1
       maskA=0
       maskB=1
       maskG=1
       maskR=1
       outputIndex=0
      }
     }
     {
      inputName=None
      input {
       expression=&self:MaterialExpressionVectorParameter_2
       inputName=None
       mask=1
       maskA=0
       maskB=1
       maskG=1
       maskR=1
       outputIndex=0
      }
     }
    ]}
  }
  MaterialExpressionTextureSampleParameter2D_0 {
   class=MaterialExpressionTextureSampleParameter2D
   inputs [
    UVs {from="", output=""}
    Apply View MipBias {from="", output=""}
   ]
   properties {
    parameterName=GrimHound_OriginalFXMask
    samplerSource=SSM_Wrap_WorldGroupSettings
    samplerType=SAMPLERTYPE_Masks
    texture=/BRCosmetics/Characters/Player/Male/Medium/Bodies/M_MED_GrimHound/Textures/T_GrimHound_Body_FX
   }
  }
  MaterialExpressionVectorParameter_0 {class=MaterialExpressionVectorParameter, properties {parameterName=GrimGoundAccent_Yellow, defaultValue=<rgba 0.7305 0.3467 0 1>}}
  MaterialExpressionVectorParameter_1 {class=MaterialExpressionVectorParameter, properties {parameterName=GrimGoundAccent_Blue, defaultValue=<rgba 0.029 0.2155 0.2899 1>}}
  MaterialExpressionVectorParameter_2 {class=MaterialExpressionVectorParameter, properties {parameterName=GrimGoundAccent_Beige, defaultValue=<rgba 0.276 0.1875 0.0887 1>}}
  MaterialExpressionVectorParameter_3 {class=MaterialExpressionVectorParameter, properties {parameterName=GrimGoundAccent_Red, defaultValue=<rgba 0.3712 0 0 1>}}
 }
 properties {
  bUseMaterialAttributes=true
  bUsedWithMeshDeformer=true
  bUsedWithMorphTargets=true
  bUsedWithSkeletalMesh=true
  pixelDepthOffsetMode=PDOM_Legacy
  shadingModel=MSM_FromMaterialExpression
  usedShadingModels=Default Lit | Subsurface | Clear Coat | Subsurface Profile | Cloth
 }
}
