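# Control Rig for the GrimHound medium male body that retargets humanoid legs onto digitigrade wolf-leg bones and adds pose-driven cloth and deform correctives.
# Each wolf thigh, calf and ball takes rest-offset rotations from human thigh, calf and foot, then FABRIK pins the toe to the animated ball.
# Remapped joint angles drive belt, left back pant-leg and hood offsets.
# Shoulder parent constraints and spherical pose readers drive upper-arm bulge and sleeve bones.
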ControlRigBlueprint /BRCosmetics/Characters/Player/Male/Medium/Bodies/M_MED_GrimHound/Meshes/M_MED_GrimHound_CtrlRig
graphs [
 RigVMFunctionLibrary {links [], nodes []}
 RigVMModel {nodes [
   RigUnit_BeginExecution = &RigUnit_BeginExecution()
   n1 = &RigUnit_GetTransform(Item=(Type=Bone,Name="dyn_wolf_thigh_l"), bInitial=true)
   n2 = &RigUnit_GetTransform(Item=(Type=Bone,Name="thigh_l"), bInitial=true)
   Set Transform = &RigUnit_SetRotation(
    Item=(Type=Bone,Name="dyn_wolf_thigh_l")
    Weight=1
    bInitial=False
    Value=&RigVMFunction_MathTransformMakeRelative(
     Global=&RigUnit_GetTransform(Item=(Type=Bone,Name="thigh_l"), Space=LocalSpace, bInitial=false).Transform
     Parent=&RigVMFunction_MathTransformMakeRelative(Parent=@n1.Transform, Global=@n2.Transform).Local
    ).Local.Rotation
    ExecutePin=@RigUnit_BeginExecution.ExecutePin
   )
   n3 = &RigUnit_GetTransform(Item=(Type=Bone,Name="dyn_wolf_calf_l"), bInitial=true)
   n4 = &RigUnit_GetTransform(Item=(Type=Bone,Name="calf_l"), bInitial=true)
   > Set Transform_1 = &RigUnit_SetRotation(
    Item=(Type=Bone,Name="dyn_wolf_calf_l")
    Weight=1
    bInitial=False
    Value=&RigVMFunction_MathTransformMakeRelative(
     Global=&RigUnit_GetTransform(Item=(Type=Bone,Name="calf_l"), Space=LocalSpace, bInitial=false).Transform
     Parent=&RigVMFunction_MathTransformMakeRelative(Parent=@n3.Transform, Global=@n4.Transform).Local
    ).Local.Rotation
   )
   n5 = &RigUnit_GetTransform(Item=(Type=Bone,Name="dyn_wolf_ball_l"), bInitial=true)
   n6 = &RigUnit_GetTransform(Item=(Type=Bone,Name="foot_l"), bInitial=true)
   > Set Transform_2 = &RigUnit_SetRotation(
    Item=(Type=Bone,Name="dyn_wolf_ball_l")
    Weight=1
    bInitial=False
    Value=&RigVMFunction_MathTransformMakeRelative(
     Global=&RigUnit_GetTransform(Item=(Type=Bone,Name="foot_l"), Space=LocalSpace, bInitial=false).Transform
     Parent=&RigVMFunction_MathTransformMakeRelative(Parent=@n5.Transform, Global=@n6.Transform).Local
    ).Local.Rotation
   )
   n7 = &RigVMFunction_MathTransformMul(
    B=&RigUnit_GetTransform(Item=(Type=Bone,Name="ball_l"), bInitial=false).Transform
    A=&RigVMFunction_MathTransformMakeRelative(
     Parent=&RigUnit_GetTransform(Item=(Type=Bone,Name="ball_l"), bInitial=true).Transform
     Global=&RigUnit_GetTransform(Item=(Type=Bone,Name="dyn_wolf_toe_l"), bInitial=true).Transform
    ).Local
   ) {template="Multiply::Execute(in A,in B,out Result)"}
   > /Script/ControlRig.RigUnit_FABRIKItemArray(
    Items=((Type=Bone,Name="dyn_wolf_thigh_l"),(Type=Bone,Name="dyn_wolf_toe_l"))
    MaxIterations=10
    Precision=1
    Weight=1
    bPropagateToChildren=True
    bSetEffectorTransform=True
    EffectorTransform=@n7.Result
   ) {is=struct RigVMUnitNode}
   > &RigUnit_SetTransform(Item=(Type=Bone,Name="dyn_wolf_toe_l"), Weight=1, bInitial=False, Value=@n7.Result)
   n8 = &RigUnit_GetTransform(Item=(Type=Bone,Name="dyn_wolf_thigh_r"), bInitial=true)
   n9 = &RigUnit_GetTransform(Item=(Type=Bone,Name="thigh_r"), bInitial=true)
   > Set Transform_3 = &RigUnit_SetRotation(
    Item=(Type=Bone,Name="dyn_wolf_thigh_r")
    Weight=1
    bInitial=False
    Value=&RigVMFunction_MathTransformMakeRelative(
     Global=&RigUnit_GetTransform(Item=(Type=Bone,Name="thigh_r"), Space=LocalSpace, bInitial=false).Transform
     Parent=&RigVMFunction_MathTransformMakeRelative(Parent=@n8.Transform, Global=@n9.Transform).Local
    ).Local.Rotation
   )
   n10 = &RigUnit_GetTransform(Item=(Type=Bone,Name="dyn_wolf_calf_r"), bInitial=true)
   n11 = &RigUnit_GetTransform(Item=(Type=Bone,Name="calf_r"), bInitial=true)
   > Set Transform_5 = &RigUnit_SetRotation(
    Item=(Type=Bone,Name="dyn_wolf_calf_r")
    Weight=1
    bInitial=False
    Value=&RigVMFunction_MathTransformMakeRelative(
     Global=&RigUnit_GetTransform(Item=(Type=Bone,Name="calf_r"), Space=LocalSpace, bInitial=false).Transform
     Parent=&RigVMFunction_MathTransformMakeRelative(Parent=@n10.Transform, Global=@n11.Transform).Local
    ).Local.Rotation
   )
   n12 = &RigUnit_GetTransform(Item=(Type=Bone,Name="dyn_wolf_ball_r"), bInitial=true)
   n13 = &RigUnit_GetTransform(Item=(Type=Bone,Name="foot_r"), bInitial=true)
   > Set Transform_6 = &RigUnit_SetRotation(
    Item=(Type=Bone,Name="dyn_wolf_ball_r")
    Weight=1
    bInitial=False
    Value=&RigVMFunction_MathTransformMakeRelative(
     Global=&RigUnit_GetTransform(Item=(Type=Bone,Name="foot_r"), Space=LocalSpace, bInitial=false).Transform
     Parent=&RigVMFunction_MathTransformMakeRelative(Parent=@n12.Transform, Global=@n13.Transform).Local
    ).Local.Rotation
   )
   n14 = &RigVMFunction_MathTransformMul(
    B=&RigUnit_GetTransform(Item=(Type=Bone,Name="ball_r"), bInitial=false).Transform
    A=&RigVMFunction_MathTransformMakeRelative(
     Global=&RigUnit_GetTransform(Item=(Type=Bone,Name="dyn_wolf_toe_r"), bInitial=true).Transform
     Parent=&RigUnit_GetTransform(Item=(Type=Bone,Name="ball_r"), bInitial=true).Transform
    ).Local
   ) {template="Multiply::Execute(in A,in B,out Result)"}
   > &RigUnit_FABRIKItemArray(
    Items=((Type=Bone,Name="dyn_wolf_thigh_r"),(Type=Bone,Name="dyn_wolf_toe_r"))
    MaxIterations=10
    Precision=1
    Weight=1
    bPropagateToChildren=True
    bSetEffectorTransform=True
    EffectorTransform=@n14.Result
   )
   > &RigUnit_SetTransform(Item=(Type=Bone,Name="dyn_wolf_toe_r"), Weight=1, bInitial=False, Value=@n14.Result)
   Absolute = &RigVMFunction_MathDoubleAbs(Value=&RigVMFunction_MathQuaternionToEuler(
     RotationOrder=ZXY
     Value=&RigUnit_GetRelativeTransformForItem(Child=(Type=Bone,Name="thigh_r"), Parent=(Type=Bone,Name="pelvis"), bChildInitial=False, bParentInitial=False).RelativeTransform.Rotation
    ).Result.Z)
   Subtract = &RigVMFunction_MathDoubleSub(A=180, B=@Absolute.Result)
   n15 = &RigUnit_GetRelativeTransformForItem(Child=(Type=Bone,Name="thigh_l"), Parent=(Type=Bone,Name="pelvis"), bChildInitial=False, bParentInitial=False)
   Maximum = &RigVMFunction_MathDoubleMax(
    A=&RigVMFunction_MathDoubleRemap(SourceMaximum=60, SourceMinimum=25, TargetMaximum=1, TargetMinimum=0, Value=@Subtract.Result).Result
    B=&RigVMFunction_MathDoubleRemap(
     SourceMaximum=-60
     SourceMinimum=-25
     TargetMaximum=1
     TargetMinimum=0
     Value=&RigVMFunction_MathQuaternionToEuler(RotationOrder=ZXY, Value=@n15.RelativeTransform.Rotation).Result.Z
    ).Result
   )
   > &RigUnit_ModifyTransforms(
    ItemToModify=((Item=(Type=Bone,Name="dyn_belt_adjustment"),Transform=(Rotation=(X=0,Y=0,Z=0,W=-1),Translation=(X=0,Y=0,Z=2.5),Scale3D=(X=0.8,Y=1,Z=0.72))))
    WeightMinimum=0
    Weight=@Maximum.Result
   )
   n16 = &RigUnit_GetRelativeTransformForItem(Child=(Type=Bone,Name="dyn_wolf_calf_l"), Parent=(Type=Bone,Name="dyn_wolf_thigh_l"), bChildInitial=False, bParentInitial=False)
   > &RigUnit_ModifyTransforms(
    ItemToModify=((Item=(Type=Bone,Name="dyn_L_pantleg_back"),Transform=(Rotation=(X=0,Y=0,Z=0.1736,W=0.9848),Translation=(X=5,Y=0,Z=0),Scale3D=(X=1,Y=1,Z=1))))
    WeightMinimum=0
    Weight=&RigVMFunction_MathDoubleRemap(
     SourceMaximum=80
     SourceMinimum=40
     TargetMaximum=1
     TargetMinimum=0
     Value=&RigVMFunction_MathQuaternionToEuler(RotationOrder=ZXY, Value=@n16.RelativeTransform.Rotation).Result.Z
    ).Result
   )
   n17 = &RigUnit_GetRelativeTransformForItem(Child=(Type=Bone,Name="neck_02"), Parent=(Type=Bone,Name="spine_05"), bChildInitial=False, bParentInitial=False)
   > &RigUnit_ModifyTransforms(
    ItemToModify=((Item=(Type=Bone,Name="dyn_hood"),Transform=(Rotation=(X=0,Y=0.0872,Z=0,W=0.9962),Translation=(X=0,Y=0,Z=0),Scale3D=(X=1,Y=1,Z=1))))
    WeightMinimum=0
    Weight=&RigVMFunction_MathDoubleRemap(
     SourceMaximum=-10
     SourceMinimum=10
     TargetMaximum=1
     TargetMinimum=0
     Value=&RigVMFunction_MathQuaternionToEuler(RotationOrder=ZXY, Value=@n17.RelativeTransform.Rotation).Result.Z
    ).Result
   )
   > &RigUnit_ParentConstraint(
    AdvancedSettings=(RotationOrderForFilter=XZY)
    Child=(Type=Bone,Name="deform_upperarm_up_base_l")
    Parents=((Item=(Type=Bone,Name="upperarm_twist_01_l"),Weight=0.5),(Item=(Type=Bone,Name="clavicle_l"),Weight=0.5))
    Weight=1
   )
   > n18 = &RigUnit_SphericalPoseReader(
    ActiveRegionScaleFactors=(PositiveWidth=2,NegativeWidth=2,PositiveHeight=2,NegativeHeight=2)
    ActiveRegionSize=0.1
    Debug=(bDrawDebug=false,DebugScale=25,DebugSegments=20,DebugThickness=0.25)
    DriverAxis=(X=1,Y=0,Z=0)
    DriverItem=(Type=Bone,Name="upperarm_l")
    FalloffRegionScaleFactors=(PositiveWidth=1,NegativeWidth=1,PositiveHeight=1,NegativeHeight=1)
    FalloffSize=0.5
    OptionalParentItem=(Type=Bone)
    RotationOffset=(X=60,Y=0,Z=0)
   )
   > &RigUnit_ModifyTransforms(
    ItemToModify=((Item=(Type=Bone,Name="deform_upperarm_up_l"),Transform=(Rotation=(X=0,Y=0,Z=0,W=1),Translation=(X=-1.6,Y=4.8,Z=4.5),Scale3D=(X=2,Y=2,Z=1.7))))
    Weight=@n18.OutputParam
   )
   > &RigUnit_ModifyTransforms(
    ItemToModify="((Item=(Type=Bone,Name=\"dyn_L_frontSleeve_support\"),Transform=(Rotation=(X=0.0151,Y=-0.0858,Z=0.173,W=0.9811),Translation=(X=0,Y=1,Z=0),Scale3D=(X=1,Y=1,Z=1))),(Item=(Type=Bone,Name=\"dyn_L_sleeve_A\"),Transform=(Rotation=(X=0,Y=0.1736,Z=0,W=0.9848),Translat ...(truncated)"
    Weight=@n18.OutputParam
   )
   > &RigUnit_ParentConstraint(
    AdvancedSettings=(RotationOrderForFilter=XZY)
    Child=(Type=Bone,Name="deform_upperarm_up_base_r")
    Parents=((Item=(Type=Bone,Name="upperarm_twist_01_r"),Weight=0.5),(Item=(Type=Bone,Name="clavicle_r"),Weight=0.5))
    Weight=1
   )
   > n19 = &RigUnit_SphericalPoseReader(
    ActiveRegionScaleFactors=(PositiveWidth=2,NegativeWidth=2,PositiveHeight=2,NegativeHeight=2)
    ActiveRegionSize=0.1
    Debug=(bDrawDebug=false,DebugScale=25,DebugSegments=20,DebugThickness=0.25)
    DriverAxis=(X=-1,Y=0,Z=0)
    DriverItem=(Type=Bone,Name="upperarm_r")
    FalloffRegionScaleFactors=(PositiveWidth=1,NegativeWidth=1,PositiveHeight=1,NegativeHeight=1)
    FalloffSize=0.5
    FlipHeightScaling=true
    OptionalParentItem=(Type=Bone)
    RotationOffset=(X=240,Y=0,Z=0)
   )
   > &RigUnit_ModifyTransforms(
    ItemToModify=((Item=(Type=Bone,Name="deform_upperarm_up_r"),Transform=(Rotation=(X=0,Y=0,Z=0,W=1),Translation=(X=1.6,Y=-4.8,Z=-4.5),Scale3D=(X=2,Y=2,Z=1.7))))
    Weight=@n19.OutputParam
   )
   > &RigUnit_ModifyTransforms(
    ItemToModify="((Item=(Type=Bone,Name=\"dyn_R_frontSleeve_support\"),Transform=(Rotation=(X=0.0226,Y=-0.0842,Z=0.2578,W=0.9623),Translation=(X=0,Y=-1,Z=0),Scale3D=(X=1,Y=1,Z=1))),(Item=(Type=Bone,Name=\"dyn_R_sleeve_A\"),Transform=(Rotation=(X=0,Y=0.1736,Z=0,W=0.9848),Transl ...(truncated)"
    Weight=@n19.OutputParam
   )
  ]}
]
hierarchy [
 root {loc=<0 0 0>, rot=<0 0 0 1>}
 pelvis {parent=root, loc=<0 -0.7071 93.1191>, rot=<0.0127 0.707 -0.0127 -0.707>}
 spine_01 {parent=pelvis, loc=<6.7494 -0.6388 0>, rot=<0 0 0.034 -0.9994>}
 spine_02 {parent=spine_01, loc=<7.2518 -0.035 0>, rot=<0 0 0.0238 -0.9997>}
 spine_03 {parent=spine_02, loc=<6.0104 0 0>, rot=<0 0 0.0149 -0.9999>}
 spine_04 {parent=spine_03, loc=<8.1024 0 0>, rot=<0 0 -0.0132 -0.9999>}
 spine_05 {parent=spine_04, loc=<10.0146 0 0>, rot=<0 0 -0.0015 -1>}
 neck_01 {parent=spine_05, loc=<11.3138 0.1529 0>, rot=<0 0 -0.2003 -0.9797>}
 neck_02 {parent=neck_01, loc=<4.5627 0 0>, rot=<0 0 0.0504 -0.9987>}
 head {parent=neck_02, loc=<5.3914 0 0>, rot=<0 0 0.115 -0.9934>}
 attach_cape {parent=spine_05, loc=<9.017 -11.714 0>, rot=<0.0616 -0.7044 -0.0616 -0.7044>}
 cape {parent=attach_cape, type=Null, loc=<0 -4.367 -0.5586>, rot=<0 0 0 1>}
 attach_backpack {parent=spine_05, loc=<-0.123 -13.473 0>, rot=<-0.0093 -0.707 0.0093 -0.707>}
 Backpack_BR {parent=attach_backpack, type=Null, loc=<0 -2.6452 0>, rot=<0 0 0 1>}
 clavicle_l {parent=spine_05, loc=<4.3575 -0.1466 -2.3223>, rot=<0.0465 -0.7289 -0.009 -0.6829>}
 upperarm_l {parent=clavicle_l, loc=<15.5111 0 0>, rot=<0.0226 -0.3234 -0.0088 -0.9459>}
 lowerarm_l {parent=upperarm_l, loc=<26.6514 0 0>, rot=<0.0391 -0.0099 -0.2453 -0.9686>}
 hand_l {parent=lowerarm_l, loc=<22.3491 0 0>, rot=<-0.7081 -0.0507 0.0975 -0.6975>}
 thumb_01_l {parent=hand_l, loc=<1.3118 -0.3191 -2.459>, rot=<0.6546 -0.4042 0.0664 -0.6354>}
 thumb_02_l {parent=thumb_01_l, loc=<5.9336 -0.0448 0.0401>, rot=<-0.0037 0.0213 0.298 -0.9543>}
 thumb_03_l {parent=thumb_02_l, loc=<2.2223 0.3682 -0.0907>, rot=<0.0118 -0.013 0.0066 -0.9998>}
 index_metacarpal_l {parent=hand_l, loc=<1.6535 0.0183 -2.96>, rot=<0.0158 -0.0124 -0.0173 -0.9996>}
 index_01_l {parent=index_metacarpal_l, loc=<8.3008 0.0357 -0.0524>, rot=<-0.0044 0.0485 0.0924 -0.9945>}
 index_02_l {parent=index_01_l, loc=<3.3587 -0.3004 0.0452>, rot=<-0.002 0.0027 0.3225 -0.9466>}
 index_03_l {parent=index_02_l, loc=<2.5578 -0.123 -0.0252>, rot=<-0.002 0.0067 0.2558 -0.9667>}
 middle_metacarpal_l {parent=hand_l, loc=<2.1856 -0.1725 -0.5984>, rot=<0.006 0.0168 -0.0329 -0.9993>}
 middle_01_l {parent=middle_metacarpal_l, loc=<7.3275 -0.2447 0.0129>, rot=<-0.0208 0.0042 0.0858 -0.9961>}
 middle_02_l {parent=middle_01_l, loc=<4.0123 0.1401 0.1175>, rot=<-0.0218 0.0258 0.3512 -0.9357>}
 middle_03_l {parent=middle_02_l, loc=<2.6077 0.0534 0.1373>, rot=<0.0082 0.0188 0.1708 -0.9851>}
 ring_metacarpal_l {parent=hand_l, loc=<2.2074 -0.3625 1.3638>, rot=<-0.0284 0.0722 -0.0458 -0.9959>}
 ring_01_l {parent=ring_metacarpal_l, loc=<7.0295 -0.6197 -0.1446>, rot=<-0.0489 -0.0252 0.1892 -0.9804>}
 ring_02_l {parent=ring_01_l, loc=<3.6174 0.4647 -0.1027>, rot=<0 0 0.2806 -0.9598>}
 ring_03_l {parent=ring_02_l, loc=<2.4152 0.114 0.0199>, rot=<0 0 0.1189 -0.9929>}
 pinky_metacarpal_l {parent=hand_l, loc=<1.7616 -0.9753 3.3814>, rot=<-0.0397 0.0814 -0.069 -0.9935>}
 pinky_01_l {parent=pinky_metacarpal_l, loc=<7.1709 -0.4831 -0.1094>, rot=<-0.0654 -0.0035 0.2242 -0.9723>}
 pinky_02_l {parent=pinky_01_l, loc=<2.8274 0.0893 -0.1043>, rot=<0.0037 0.0011 0.2854 -0.9584>}
 pinky_03_l {parent=pinky_02_l, loc=<2.2046 0.0307 -0.0078>, rot=<-0.0002 0.0098 0.1445 -0.9895>}
 lowerarm_twist_02_l {parent=lowerarm_l, loc=<9.6199 0 0>, rot=<0 0 0 -1>}
 lowerarm_twist_01_l {parent=lowerarm_l, loc=<17.8549 0 0>, rot=<0 0 0 -1>}
 upperarm_twist_01_l {parent=upperarm_l, loc=<0.0041 0 0>, rot=<0 0 0 -1>}
 dyn_L_shoulderFur_A {parent=upperarm_twist_01_l, loc=<7.9414 -1.543 8.0377>, rot=<0.9236 -0.0783 -0.2951 -0.2319>}
 dyn_L_shoulderFur_B {parent=upperarm_twist_01_l, loc=<3.5447 -2.2039 8.3318>, rot=<0.9622 -0.1644 -0.0882 -0.1985>}
 dyn_L_shoulderFur_C {parent=upperarm_twist_01_l, loc=<5.6486 -3.7511 7.9743>, rot=<0.9257 -0.1296 -0.2418 -0.2602>}
 dyn_L_shoulderFur_D {parent=upperarm_twist_01_l, loc=<7.9458 -5.637 5.4807>, rot=<0.8214 -0.1247 -0.4226 -0.3623>}
 dyn_L_shoulderFur_E {parent=upperarm_twist_01_l, loc=<2.005 -5.3873 7.4583>, rot=<0.9196 0.0212 -0.2506 -0.3019>}
 deform_upperarm_up_base_l {parent=upperarm_twist_01_l, loc=<-3.2216 -1.1391 3.3311>, rot=<0 0 0 1>}
 deform_upperarm_up_l {parent=deform_upperarm_up_base_l, loc=<0 0 0>, rot=<0 0 0 1>}
 upperarm_twist_02_l {parent=upperarm_l, loc=<13.7941 0 0>, rot=<0 0 0 -1>}
 dyn_L_sleeve_A {parent=clavicle_l, loc=<14.3983 0.3667 7.7268>, rot=<0.7808 0.04 0.0644 -0.6201>}
 dyn_L_sleeve_B {parent=clavicle_l, loc=<14.7314 -9.4764 4.0878>, rot=<0.2653 0.0841 -0.0732 -0.9577>}
 dyn_L_frontSleeve_support {parent=clavicle_l, loc=<7.1405 9.2028 0.1528>, rot=<0 0 0 1>}
 clavicle_r {parent=spine_05, loc=<4.3575 -0.1466 2.3223>, rot=<0.7289 0.0465 -0.6829 0.009>}
 upperarm_r {parent=clavicle_r, loc=<-15.5111 0 0>, rot=<0.0226 -0.3234 -0.0088 -0.9459>}
 lowerarm_r {parent=upperarm_r, loc=<-26.6514 0 0>, rot=<0.0391 -0.0099 -0.2453 -0.9686>}
 hand_r {parent=lowerarm_r, loc=<-22.3491 0 0>, rot=<-0.7081 -0.0507 0.0975 -0.6975>}
 thumb_01_r {parent=hand_r, loc=<-1.3118 0.3191 2.459>, rot=<0.6546 -0.4042 0.0664 -0.6354>}
 thumb_02_r {parent=thumb_01_r, loc=<-5.9336 0.0448 -0.0401>, rot=<-0.0037 0.0213 0.298 -0.9543>}
 thumb_03_r {parent=thumb_02_r, loc=<-2.2223 -0.3682 0.0907>, rot=<0.0118 -0.013 0.0066 -0.9998>}
 index_metacarpal_r {parent=hand_r, loc=<-1.6535 -0.0183 2.96>, rot=<0.0158 -0.0124 -0.0173 -0.9996>}
 index_01_r {parent=index_metacarpal_r, loc=<-8.3008 -0.0357 0.0524>, rot=<-0.0044 0.0485 0.0924 -0.9945>}
 index_02_r {parent=index_01_r, loc=<-3.3587 0.3004 -0.0452>, rot=<-0.002 0.0027 0.3225 -0.9466>}
 index_03_r {parent=index_02_r, loc=<-2.5578 0.123 0.0252>, rot=<-0.002 0.0067 0.2558 -0.9667>}
 middle_metacarpal_r {parent=hand_r, loc=<-2.1856 0.1725 0.5984>, rot=<0.006 0.0168 -0.0329 -0.9993>}
 middle_01_r {parent=middle_metacarpal_r, loc=<-7.3275 0.2447 -0.0129>, rot=<-0.0208 0.0042 0.0858 -0.9961>}
 middle_02_r {parent=middle_01_r, loc=<-4.0123 -0.1401 -0.1175>, rot=<-0.0218 0.0258 0.3512 -0.9357>}
 middle_03_r {parent=middle_02_r, loc=<-2.6077 -0.0534 -0.1373>, rot=<0.0082 0.0188 0.1708 -0.9851>}
 ring_metacarpal_r {parent=hand_r, loc=<-2.2074 0.3625 -1.3638>, rot=<-0.0284 0.0722 -0.0458 -0.9959>}
 ring_01_r {parent=ring_metacarpal_r, loc=<-7.0295 0.6197 0.1446>, rot=<-0.0489 -0.0252 0.1892 -0.9804>}
 ring_02_r {parent=ring_01_r, loc=<-3.6174 -0.4647 0.1027>, rot=<0 0 0.2806 -0.9598>}
 ring_03_r {parent=ring_02_r, loc=<-2.4152 -0.114 -0.0199>, rot=<0 0 0.1189 -0.9929>}
 pinky_metacarpal_r {parent=hand_r, loc=<-1.7616 0.9753 -3.3814>, rot=<-0.0397 0.0814 -0.069 -0.9935>}
 pinky_01_r {parent=pinky_metacarpal_r, loc=<-7.1709 0.4831 0.1094>, rot=<-0.0654 -0.0035 0.2242 -0.9723>}
 pinky_02_r {parent=pinky_01_r, loc=<-2.8274 -0.0893 0.1043>, rot=<0.0037 0.0011 0.2854 -0.9584>}
 pinky_03_r {parent=pinky_02_r, loc=<-2.2046 -0.0307 0.0078>, rot=<-0.0002 0.0098 0.1445 -0.9895>}
 lowerarm_twist_02_r {parent=lowerarm_r, loc=<-9.6199 0 0>, rot=<0 0 0 -1>}
 lowerarm_twist_01_r {parent=lowerarm_r, loc=<-17.8549 0 0>, rot=<0 0 0 -1>}
 upperarm_twist_01_r {parent=upperarm_r, loc=<-0.0041 0 0>, rot=<0 0 0 -1>}
 dyn_R_shoulderFur_A {parent=upperarm_twist_01_r, loc=<-7.9415 1.543 -8.0376>, rot=<0.9236 -0.0783 -0.2951 -0.2319>}
 dyn_R_shoulderFur_B {parent=upperarm_twist_01_r, loc=<-3.4316 2.2039 -8.3794>, rot=<0.9628 -0.1631 -0.0817 -0.1996>}
 dyn_R_shoulderFur_C {parent=upperarm_twist_01_r, loc=<-5.6488 3.7511 -7.9741>, rot=<0.9257 -0.1296 -0.2418 -0.2602>}
 dyn_R_shoulderFur_D {parent=upperarm_twist_01_r, loc=<-7.9456 5.637 -5.481>, rot=<0.8214 -0.1247 -0.4226 -0.3623>}
 dyn_R_shoulderFur_E {parent=upperarm_twist_01_r, loc=<-2.0047 5.3873 -7.4587>, rot=<0.9196 0.0212 -0.2506 -0.3019>}
 deform_upperarm_up_base_r {parent=upperarm_twist_01_r, loc=<3.2214 1.1391 -3.3308>, rot=<0 0 0 1>}
 deform_upperarm_up_r {parent=deform_upperarm_up_base_r, loc=<0 0 0>, rot=<0 0 0 -1>}
 upperarm_twist_02_r {parent=upperarm_r, loc=<-13.7941 0 0>, rot=<0 0 0 -1>}
 dyn_R_sleeve_A {parent=clavicle_r, loc=<-14.3983 -0.3667 -7.7266>, rot=<0.7808 0.04 0.0644 -0.6201>}
 dyn_R_sleeve_B {parent=clavicle_r, loc=<-14.7314 9.4763 -4.0878>, rot=<0.2653 0.0841 -0.0732 -0.9577>}
 dyn_R_frontSleeve_support {parent=clavicle_r, loc=<-7.1405 -9.2028 -0.1527>, rot=<0 0 0 1>}
 dyn_hood {parent=spine_05, loc=<13.3136 2.1385 -0.0284>, rot=<0.5694 -0.4193 0.4197 -0.569>}
 thigh_l {parent=pelvis, loc=<-2.5693 -0.0109 -7.7501>, rot=<0.0012 0.0572 0.0155 -0.9982>}
 calf_l {parent=thigh_l, loc=<-39.0484 0 0>, rot=<0.0006 -0.0161 -0.0391 -0.9991>}
 foot_l {parent=calf_l, loc=<-42.561 0.1077 0.4292>, rot=<0.0418 -0.0455 0.0343 -0.9975>}
 ball_l {parent=foot_l, loc=<-6.1599 13.8264 0.6806>, rot=<0.0001 -0.0001 -0.7063 -0.7079>}
 calf_twist_02_l {parent=calf_l, loc=<-18.8064 -0.0967 0.0397>, rot=<0.0237 0.0095 -0.0035 -0.9997>}
 calf_twist_01_l {parent=calf_l, loc=<-35.813 -0.0967 0.0397>, rot=<0.0237 0.0095 -0.0035 -0.9997>}
 thigh_twist_01_l {parent=thigh_l, loc=<-21.6303 0 0>, rot=<0.0013 0 0 -1>}
 thigh_r {parent=pelvis, loc=<-2.5693 -0.0109 7.7501>, rot=<0.0572 -0.0012 0.9982 0.0155>}
 calf_r {parent=thigh_r, loc=<39.0484 0 0>, rot=<0.0006 -0.0161 -0.0391 -0.9991>}
 foot_r {parent=calf_r, loc=<42.561 -0.1077 -0.4292>, rot=<0.0418 -0.0455 0.0343 -0.9975>}
 ball_r {parent=foot_r, loc=<6.1599 -13.8264 -0.6806>, rot=<0.0001 -0.0001 -0.7063 -0.7079>}
 calf_twist_02_r {parent=calf_r, loc=<18.8064 0.0967 -0.0397>, rot=<0.0237 0.0095 -0.0035 -0.9997>}
 calf_twist_01_r {parent=calf_r, loc=<35.813 0.0967 -0.0397>, rot=<0.0237 0.0095 -0.0035 -0.9997>}
 thigh_twist_01_r {parent=thigh_r, loc=<21.6303 0 0>, rot=<0.0013 0 0 -1>}
 dyn_wolf_thigh_l {parent=pelvis, loc=<-2.5693 -0.0109 -7.7501>, rot=<0.0012 0.0572 0.0155 -0.9982>}
 dyn_wolf_calf_l {parent=dyn_wolf_thigh_l, loc=<-36.3283 2.1181 0.0962>, rot=<0.0006 -0.0161 -0.0391 -0.9991>}
 dyn_wolf_calf_twist_02_l {parent=dyn_wolf_calf_l, loc=<-10.4852 -4.0769 0.3061>, rot=<0.0237 0.0095 -0.0035 -0.9997>}
 dyn_wolf_foot_l {parent=dyn_wolf_calf_l, loc=<-23.7181 -7.4961 0.2405>, rot=<0.0418 -0.0455 0.0343 -0.9975>}
 dyn_wolf_ball_l {parent=dyn_wolf_foot_l, loc=<-19.3349 3.4907 -0.8786>, rot=<0.0001 -0.0001 -0.7063 -0.7079>}
 dyn_wolf_toe_l {parent=dyn_wolf_ball_l, loc=<8.8736 8.7423 -1.1808>, rot=<0 0 0 1>}
 dyn_wolf_thigh_twist_01_l {parent=dyn_wolf_thigh_l, loc=<-21.6303 0 0>, rot=<0.0013 0 0 -1>}
 dyn_L_pantleg_front {parent=dyn_wolf_thigh_l, loc=<-30.811 3.8834 0.2022>, rot=<0 0 0 1>}
 dyn_L_pantleg_back {parent=dyn_wolf_thigh_l, loc=<-30.8248 1.054 0.2018>, rot=<0 0 0 1>}
 dyn_L_leg_panel {parent=dyn_wolf_thigh_l, loc=<-5.7934 0.6101 -6.1603>, rot=<-0.0244 -0.991 0.0197 -0.1304>}
 dyn_wolf_thigh_r {parent=pelvis, loc=<-2.5694 -0.0109 7.7501>, rot=<0.0572 -0.0012 0.9982 0.0155>}
 dyn_wolf_calf_r {parent=dyn_wolf_thigh_r, loc=<36.3283 -2.1181 -0.0962>, rot=<0.0006 -0.0161 -0.0391 -0.9991>}
 dyn_wolf_calf_twist_02_r {parent=dyn_wolf_calf_r, loc=<10.4852 4.0769 -0.3061>, rot=<0.0237 0.0095 -0.0035 -0.9997>}
 dyn_wolf_foot_r {parent=dyn_wolf_calf_r, loc=<23.7181 7.4961 -0.2404>, rot=<0.0418 -0.0455 0.0343 -0.9975>}
 dyn_wolf_ball_r {parent=dyn_wolf_foot_r, loc=<19.3349 -3.4907 0.8786>, rot=<0.0001 -0.0001 -0.7063 -0.7079>}
 dyn_wolf_toe_r {parent=dyn_wolf_ball_r, loc=<-8.8736 -8.7424 1.1809>, rot=<0 0 0 1>}
 dyn_R_pantleg_1 {parent=dyn_wolf_calf_r, loc=<2.5002 -6.582 1.6162>, rot=<0.7681 0.1041 -0.1319 -0.6179>}
 dyn_R_pantleg_2 {parent=dyn_R_pantleg_1, loc=<7.3563 1.1225 0.6868>, rot=<-0.2259 -0.1524 0.0108 -0.9621>}
 wolf_thigh_twist_01_r {parent=dyn_wolf_thigh_r, loc=<21.6303 0 0>, rot=<0.0013 0 0 -1>}
 dyn_R_leg_panel {parent=dyn_wolf_thigh_r, loc=<5.7933 -0.6101 6.1603>, rot=<-0.0244 -0.991 0.0197 -0.1304>}
 dyn_butt_panel {parent=pelvis, loc=<3.1598 -10.9261 0>, rot=<0.0534 -0.9986 0 0>}
 dyn_belt_adjustment {parent=pelvis, loc=<5.4015 10.7447 0>, rot=<-0.0127 -0.707 0.0127 -0.707>}
 dyn_nunchucks_base {parent=pelvis, loc=<7.1608 -4.3309 12.3256>, rot=<-0.0127 -0.707 0.0127 -0.707>}
 dyn_nunchuck_A_master {parent=dyn_nunchucks_base, loc=<-1.8806 1.1945 -0.7339>, rot=<-0.373 0.3316 0.8432 0.1997>}
 dyn_nunchuck_A_1 {parent=dyn_nunchuck_A_master, loc=<0 0 0>, rot=<0 0 0 1>}
 dyn_nunchuck_A_2 {parent=dyn_nunchuck_A_1, loc=<1.1332 0 0>, rot=<0.1239 -0.188 0.0046 -0.9743>}
 dyn_nunchuck_A_3 {parent=dyn_nunchuck_A_2, loc=<1.0897 0 0>, rot=<0.0156 -0.0002 0.0206 -0.9997>}
 dyn_nunchuck_A_4 {parent=dyn_nunchuck_A_3, loc=<1.1781 0 0>, rot=<0.0423 0.0004 0.0392 -0.9983>}
 dyn_nunchuck_B_master {parent=dyn_nunchucks_base, loc=<-0.6342 -1.9509 -3.4907>, rot=<-0.346 -0.2246 0.7642 -0.4959>}
 dyn_nunchuck_B_1 {parent=dyn_nunchuck_B_master, loc=<0 0 0>, rot=<0 0 0 1>}
 dyn_nunchuck_B_2 {parent=dyn_nunchuck_B_1, loc=<2.23 0 0>, rot=<0.0124 -0.1654 0.1037 -0.9807>}
 LegAdjust {type=Curve, loc=<0 0 0>, rot=<0 0 0 1>}
]
