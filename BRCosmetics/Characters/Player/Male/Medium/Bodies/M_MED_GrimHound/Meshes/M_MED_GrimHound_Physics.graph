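# Physics asset for the GrimHound medium male player body, supplying collision and joints for its dangling nunchuck bones.
# Two nunchuck chains run from their master bones through small capsule links to longer end capsules, joined by angle-limited constraints that loosen at the last joint.
# Larger capsules on the right leg panel and right wolf thigh, plus a small pelvis capsule, have no constraints and likely act as colliders for the chains.
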PhysicsAsset /BRCosmetics/Characters/Player/Male/Medium/Bodies/M_MED_GrimHound/Meshes/M_MED_GrimHound_Physics
bodies {
 dyn_R_leg_panel [{
   shapeType=Capsule
   length=20
   radius=18
   shapeName=dyn_R_leg_panel_capsule
   center=<-2.7856 -0.2609 15.8766>
   rotation=<rot 85 -0.0001 -0.0003>
  }]
 dyn_nunchuck_A_1 [{shapeType=Capsule, length=1, radius=1, shapeName=dyn_nunchuck_A_1_capsule, center=<1 0 0>, rotation=<rot 90 0 0>}]
 dyn_nunchuck_A_2 [{shapeType=Capsule, length=1, radius=1, shapeName=dyn_nunchuck_A_2_capsule, center=<1 0 0>, rotation=<rot 90 0 0>}]
 dyn_nunchuck_A_3 [{shapeType=Capsule, length=1, radius=1, shapeName=dyn_nunchuck_A_3_capsule, center=<1 0 0>, rotation=<rot 90 0 0>}]
 dyn_nunchuck_A_4 [{shapeType=Capsule, length=22, radius=2, shapeName=dyn_nunchuck_A_4_capsule, center=<14 0 0>, rotation=<rot 90 0 0>}]
 dyn_nunchuck_A_master [{shapeType=Capsule, length=0.505, radius=0.505, shapeName=dyn_nunchuck_A_master_capsule, rotation=<rot 0 -90 0>}]
 dyn_nunchuck_B_1 [{shapeType=Capsule, length=1, radius=1, shapeName=dyn_nunchuck_B_1_capsule, center=<1 0 0>, rotation=<rot 90 0 0>}]
 dyn_nunchuck_B_2 [{shapeType=Capsule, length=22, radius=2, shapeName=dyn_nunchuck_A_4_capsule, center=<14 0 -1.5>, rotation=<rot 80 0 0>}]
 dyn_nunchuck_B_master [{shapeType=Capsule, length=0.505, radius=0.505, shapeName=dyn_nunchuck_B_master_capsule, rotation=<rot 0 -90 0>}]
 dyn_wolf_thigh_r [{shapeType=Capsule, length=27, radius=17, shapeName=dyn_wolf_thigh_r_capsule, center=<20.9244 -1.7684 -8.6906>, rotation=<rot 90 0 0>}]
 pelvis [{shapeType=Capsule, length=1, radius=1, shapeName=pelvis_capsule, center=<-2.4214 -0.3557 -0.1546>}]
}
constraints [
 {bone1Name=dyn_nunchuck_A_1, bone2Name=dyn_nunchuck_A_master, swing1LimitDegrees=20, swing2LimitDegrees=20, twistLimitDegrees=20}
 {bone1Name=dyn_nunchuck_A_2, bone2Name=dyn_nunchuck_A_1, swing1LimitDegrees=20, swing2LimitDegrees=20, twistLimitDegrees=20}
 {bone1Name=dyn_nunchuck_A_3, bone2Name=dyn_nunchuck_A_2, swing1LimitDegrees=20, swing2LimitDegrees=20, twistLimitDegrees=20}
 {bone1Name=dyn_nunchuck_A_4, bone2Name=dyn_nunchuck_A_3, swing1LimitDegrees=30, swing2LimitDegrees=30, twistLimitDegrees=30}
 {bone1Name=dyn_nunchuck_B_1, bone2Name=dyn_nunchuck_B_master, swing1LimitDegrees=20, swing2LimitDegrees=20, twistLimitDegrees=20}
 {bone1Name=dyn_nunchuck_B_2, bone2Name=dyn_nunchuck_B_1, swing1LimitDegrees=30, swing2LimitDegrees=30, twistLimitDegrees=30}
]
